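# Translucent unlit material for scanline mesh particles in the Wayfare scan emote FX, built for Niagara.
# SplineThicken takes its base and tip width from a Thickness dynamic parameter, so particles control the line width.
# Particle color RGB feeds the shared emissive adjustment function.
# A fade mask samples an electronic gradient along U mirrored around the center, sharpens it with a power, then multiplies by particle alpha and linear V.
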Material /BRCosmetics/Animation/Game/MainPlayer/Emotes/WayFare/FX/M_WayfareScan_Scanline
graph {
 properties {
  bUsedWithNiagaraMeshParticles=true
  blendMode=BLEND_Translucent
  pixelDepthOffsetMode=PDOM_Legacy
  shadingModel=MSM_Unlit
  usedShadingModels=Unlit
 }
 nodes [
  n1 = DynamicParameter() {paramNames [Thickness,Param2,Param3,Param4]}
  MaterialFunctionCall(WidthBase=@n1.Thickness, WidthTip=@n1.Thickness) {materialFunction=&SplineThicken}
  n2 = ParticleColor()
  MaterialFunctionCall(Emissive=@n2.RGB) {materialFunction=&MF_FX_EmissiveAdjustments}
  n3 = ComponentMask(None=TextureCoordinate() {uTiling=2}) {r=true}
  n4 = Constant() {r=1}
  n5 = TextureSample(UVs=OneMinus(None=Distance(A=@n3, B=@n4))) {texture=/BRCosmetics/Weapons/WeaponSkins/Textures/ElectronicGradient1}
  Multiply(A=ComponentMask(None=TextureCoordinate() {uTiling=2}) {g=true}, B=Multiply(A=Power(Base=@n5.G), B=@n2.A))
  TextureCoordinate()
 ]
}
