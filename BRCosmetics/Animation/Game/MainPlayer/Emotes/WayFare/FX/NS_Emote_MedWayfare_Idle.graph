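# Niagara system for the MedWayfare emote idle that draws a blue eye glow.
# Its EyeGlow emitter bursts one sprite lasting the emitter loop, with two renderers and soft-glow size and color parameters.
# System.ViewConeFalloffEval lerps the camera offset to pull the glow forward by view angle.
# Alpha scales with time of day and fades out as the owner's X axis nears vertical.
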NiagaraSystem /BRCosmetics/Animation/Game/MainPlayer/Emotes/WayFare/FX/NS_Emote_MedWayfare_Idle
dependencies {
 usedDataInterfaces []
 usedDynamicInputs [&self:DotAgainstCameraFwd]
 usedModules [
  &self:SystemSpawnScript.SystemScriptSource.SystemScriptGraph.NiagaraNodeAssignment_1.SetVariables_#1
  &self:SystemSpawnScript.SystemScriptSource.SystemScriptGraph.NiagaraNodeAssignment_0.SetVariables_#2
  &SystemState
 ]
}
emitters {EyeGlow {
  emitterUpdateScript [
   &EmitterState(Inactive Response=::NewEnumerator1, Life Cycle Mode=::NewEnumerator1, Loop Behavior=::NewEnumerator1, Loop Duration=20, UseLoopDelay=false)
   &SpawnBurst_Instantaneous(Spawn Count=1, Use Spawn Probability=false)
  ]
  particleSpawnScript [
   &InitializeParticle(
    Color=<rgba 0.3728 0.6757 1.5 1>
    Lifetime=$Emitter.CurrentLoopDuration
    Sprite Size Mode=::NewEnumerator3
    Uniform Sprite Size=30
    UsePositionOffset=false
    Write Material Random=true
   )
   /SetVariables(Particles.SoftGlowSize:Vector2f=<35 35>) {isSetParameters=true}
   /SetVariables(Particles.SoftGlowColor:LinearColor=&MultiplyLinearColor(Linear Color=$Particles.Color, Scale Factor=<1 1 1 0.025>)) {isSetParameters=true}
   &Color(Color=&MultiplyLinearColor(
     Linear Color=$Particles.Color
     Scale Factor=&MakeVector4_FromFloat(Float=&FN_Select_Float_from_Time_Of_Day(Morning=0.3, Day=0.75, Evening=0.7, Night=0.7))
    )) {enabled=false, module=Color001}
  ]
  particleUpdateScript [
   &ParticleState()
   &SolveForcesAndVelocity(Clamp Velocity=false, Limit Acceleration=false)
   /SetVariables(Particles.SoftGlowColor:LinearColor=&MultiplyLinearColor(Linear Color=$Particles.Color, Scale Factor=<1 1 1 0.025>)) {isSetParameters=true}
   &CameraOffset(Camera Offset Amount=&Lerp_Float(A=0, B=11, Alpha=$System.ViewConeFalloffEval))
   &DynamicMaterialParameters(
    Index 0 Param 1=20
    Index 0 Param 2=0.1
    Index 1 Param 1=2
    Module.Param0WriteEnabled=true
    Module.Param0WriteEnabled1=true
    Module.Param1WriteEnabled=true
    Module.Param1WriteEnabled1=true
    Module.Param2WriteEnabled=true
    Module.Param2WriteEnabled1=true
    Module.Param3WriteEnabled=true
    Module.Param3WriteEnabled1=true
    Write Parameter Index 1=true
   )
   &Color(Color=&LerpLinearColors(StartColor=<rgba 1 0 0 10>, EndColor=<rgba 0 1 0 10>, LerpFactor=$System.ViewConeFalloffEval)) {enabled=false}
   &ScaleColor(
    Scale Alpha=&Multiply_Float(
     A=&Multiply_Float(
      A=&FN_Select_Float_from_Time_Of_Day(Morning=0.3, Day=0.7, Evening=0.7, Night=0.7)
      B=&RemapRange(
       Input Value To Remap=&Abs_Float(Float:NiagaraFloat=&DotProduct(Vector A=$Engine.Owner.SystemXAxis, Vector B=<0 0 1>))
       Input Min=0.75
       Input Max=1
       Output Min=1
       Output Max=0
      )
     )
     B=0.5
    )
    ScaleA=true
    ScaleRGB=false
   )
   /SetVariables(Particles.SoftGlowColor:LinearColor=&MultiplyLinearColor(Linear Color=$Particles.Color, Scale Factor=<1 1 1 0.025>)) {isSetParameters=true}
  ]
  rendererClasses [&NiagaraSpriteRendererProperties]
  renderers [
   {rendererClass=&NiagaraSpriteRendererProperties}
   {rendererClass=&NiagaraSpriteRendererProperties, rendererIndex=1}
  ]
 }}
summary {
 systemName=NS_Emote_MedWayfare_Idle
 emitters [{emitterName=EyeGlow, rendererClasses [&NiagaraSpriteRendererProperties]}]
 userVariables []
}
