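# Niagara effect for the MedWayfare scanning emote: a flickering blue torch glow with animated scan lines.
# ScanningLines bursts mesh lines with sprites, fanning and spinning them via per-particle quaternions from normalized index, with wave-modulated length.
# TorchGlow's two mesh renderers follow system reveal and find timelines, scaling brightness by time of day and front-end preview.
# ScanResult, a burst of sprites pulled in by point attraction, is disabled.
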NiagaraSystem /BRCosmetics/Animation/Game/MainPlayer/Emotes/WayFare/FX/NS_Emote_MedWayfare_Scanning
dependencies {
 usedDataInterfaces [&NiagaraDataInterfaceCurve, &NiagaraDataInterfaceColorCurve, &NiagaraDataInterfaceArrayColor]
 usedDynamicInputs [&ClampFloat, &RemapRange, &Subtract_Float]
 usedModules [
  &self:SystemSpawnScript.SystemScriptSource.SystemScriptGraph.NiagaraNodeAssignment_2.SetVariables_#1
  &self:SystemSpawnScript.SystemScriptSource.SystemScriptGraph.NiagaraNodeAssignment_3.SetVariables_#2
  &self:SystemSpawnScript.SystemScriptSource.SystemScriptGraph.NiagaraNodeAssignment_1.SetVariables_#3
  &self:SystemSpawnScript.SystemScriptSource.SystemScriptGraph.NiagaraNodeAssignment_5.SetVariables_#4
  &SystemState
  &self:SystemSpawnScript.SystemScriptSource.SystemScriptGraph.NiagaraNodeAssignment_4.SetVariables_#5
  &self:SystemSpawnScript.SystemScriptSource.SystemScriptGraph.NiagaraNodeAssignment_0.SetVariables_#6
 ]
}
emitters {
 ScanResult {
  bEnabled=false
  emitterUpdateScript [
   &EmitterState()
   &SpawnBurst_Instantaneous(Spawn Count=50, Spawn Time=$System.find_burst_time, Use Spawn Probability=false, UseLoopCountLimit=false)
  ]
  particleSpawnScript [
   &InitializeParticle(
    Color=&SelectLinearColorFromArray(
     Array Sampling Mode=::NewEnumerator0
     Direct Array Index=?
     Color Selection Array=DataInterface() {propertyValues {ColorData [<rgba 0.1891 1.305 2 1>, <rgba 0 0.3552 3 1>]}}
     Random Seed=0
    )
    Lifetime Max=1
    Lifetime Min=0.5
    Lifetime Mode=::NewEnumerator1
    Sprite Rotation Mode=::NewEnumerator2
    Sprite Size Mode=::NewEnumerator1
    Uniform Sprite Size Max=1
    Uniform Sprite Size Min=1
    UsePositionOffset=false
    Write Material Random=true
   )
   /Niagara/Modules/Update/Mass/CalculateSizeAndRotationalInertiaByMass(Module.Manually Enter Density=false)
   &SphereLocation(
    MaskSpawnGroup=false
    Offset=<120 0 0>
    Override Seed=false
    Sphere Radius=15
    Surface Only=false
    UseNonUniformScale=false
    UseOffset=true
    UseSphereOrientation=false
   )
   /SetVariables(Particles.SpriteSize=&Multiply_Vector2D_ByFloat(Float=&RandomRangeFloat(Minimum=2, Maximum=3), Vector2D=<1 1>)) {isSetParameters=true}
   &AddVelocityFromPoint(
    Module.Use Falloff Distance=false
    Origin Offset=$Output.SphereLocation.Position
    Velocity Origin:Vector3f=/Niagara/DynamicInputs/Helpers/SimulationPosition()
    Velocity Strength=&RandomRangeFloat(Minimum=0, Maximum=50)
   )
  ]
  particleUpdateScript [
   &ParticleState()
   &Drag(Drag=3, Use Linear Drag=true, Use Rotational Drag=false)
   &ScaleSpriteSize(Scale Factor=&Vector2DFromFloat(Value=&FloatFromCurve(FloatCurve=DataInterface() {propertyValues {Curve {keys [
          <at -0.0046 1.0113 RCIM_Cubic>
          {arriveTangent=-1.0036, interpMode=RCIM_Cubic, leaveTangent=-1.0036, time=0.7001, value=0.872}
          <at 1.0011 0.002>
         ]}}})))
   &ScaleColor(
    Scale Alpha=&FloatFromCurve(FloatCurve=DataInterface() {propertyValues {Curve {keys [
         <at 0 0 RCIM_Cubic>
         {arriveTangent=1.4286, interpMode=RCIM_Cubic, leaveTangent=1.4286, time=0.3, value=0.975}
         {arriveTangent=-1.3929, interpMode=RCIM_Cubic, leaveTangent=-1.3929, time=0.7, value=1}
         <at 1 0 RCIM_Cubic>
        ]}}})
    ScaleA=true
    ScaleRGB=true
   )
   &PointAttractionForce(
    Attraction Radius=125
    Attraction Strength=&RandomRangeFloat(Minimum=30, Maximum=250)
    Falloff Exponent=0.35
    Kill Radius=1
    Kill Within Radius=true
    Module.Use Falloff=true
    UseAttractorOffset=false
   )
   &SolveForcesAndVelocity(Clamp Velocity=false, Limit Acceleration=false)
   &ScaleSpriteSizeBySpeed(Max Scale Factor=<1 8>, Velocity Threshold=500)
  ]
  rendererClasses [&NiagaraSpriteRendererProperties]
  renderers [{rendererClass=&NiagaraSpriteRendererProperties}]
 }
 ScanningLines {
  emitterSpawnScript [/SetVariables(Emitter.ScanWidth=100) {isSetParameters=true}]
  emitterUpdateScript [
   &EmitterState()
   &SpawnBurst_Instantaneous(Spawn Count=$System.NumScanlines, Use Spawn Probability=false, UseLoopCountLimit=false)
   /SetVariables(Emitter.ScanWidth:NiagaraFloat=&Lerp_Float(A=0, B=10, Alpha=$System.ScanWidthTime)) {isSetParameters=true}
  ]
  particleSpawnScript [
   &InitializeParticle(
    Color Mode=::NewEnumerator0
    Lifetime=$Emitter.CurrentLoopDuration
    Mesh Scale Mode=::NewEnumerator3
    Sprite Rotation Mode=::NewEnumerator2
    Sprite Size Mode=::NewEnumerator3
    Uniform Sprite Size=15
    UsePositionOffset=false
    Write Material Random=true
   )
   /SetVariables(Particles.Line_Normalized_Index:NiagaraFloat=&ReturnNormalizedExecIndex(Normalized Index Includes One=false)) {isSetParameters=true}
   &Color()
   /SetVariables(Particles.CalculatedLineOrientation:Quat4f=&MakeQuatFromAxisAngle(
     Axis=&MakeVector(
      X=0
      Y=&Sine(Normalized Angle=$Particles.Line_Normalized_Index, Period=1, Scale=1, Bias=0)
      Z=&Cosine(Normalized Angle=$Particles.Line_Normalized_Index, Period=1, Scale=1, Bias=0)
     )
     AngleInDegrees=16.0452
    )) {isSetParameters=true}
   /SetVariables() {isSetParameters=true}
   &InitialMeshOrientation(Mesh Orientation Mode=::NewEnumerator6, Orientation Quaternion:Quat4f=$Particles.CalculatedLineOrientation, Use Rotation Vector=false)
   /SetVariables(Particles.Line_Length:NiagaraFloat=&RandomRangeFloat(Minimum=50, Maximum=100)) {isSetParameters=true}
  ]
  particleUpdateScript [
   &ParticleState()
   &DynamicMaterialParameters(
    Index 0 Param 1=5
    Module.Param0WriteEnabled=true
    Module.Param1WriteEnabled=true
    Module.Param2WriteEnabled=true
    Module.Param3WriteEnabled=true
   )
   /SetVariables(Particles.CalculatedLineOrientation:Quat4f=&MultiplyQuaternion(
     Quaternion A=&MakeQuatFromAxisAngle(
      Axis=<1 0 0>
      AngleInDegrees=&Multiply_Float(
       A=&Add_Float(A=$Emitter.NormalizedLoopAge, B=&RandomRangeFloat(Minimum=-0.2, Maximum=0.2))
       B=&RandomRangeFloat(Minimum=900, Maximum=-1080)
      )
     )
     Quaternion B=&MultiplyQuaternion(
      Quaternion A=&MakeQuatFromAxisAngle(
       Axis=&MakeVector(
        X=1
        Y=&Sine(Normalized Angle=$Particles.Line_Normalized_Index, Period=1, Scale=1, Bias=0)
        Z=&Cosine(Normalized Angle=$Particles.Line_Normalized_Index, Period=1, Scale=1, Bias=0)
       )
       AngleInDegrees=$Emitter.ScanWidth
      )
      Quaternion B=<0 0 0 1>
     )
    )) {isSetParameters=true}
   /SetVariables(Particles.Scale=&MakeVector(X=&Multiply_Float(A=$Particles.Line_Length, B=0.01), Y=1, Z=1)) {isSetParameters=true}
   /SetVariables(Particles.dest_length:NiagaraFloat=&Multiply_Float(
     A=&Lerp_Float(
      A=100
      B=125
      Alpha=&Waveform(
       Global Phase=&Add_Float(A=$Emitter.NormalizedLoopAge, B=&Multiply_Float(A=$Particles.Line_Normalized_Index, B=20))
       "Waveform [1]"=::NewEnumerator12
       "[1] Phase"=&Add_Float(A=$Emitter.NormalizedLoopAge, B=&Multiply_Float(A=$Particles.Line_Normalized_Index, B=20))
       "[1] Frequency"=8
       "[1] Amplitude Min/Max"=<0 1>
      )
     )
     B=&FloatFromCurve(
      FloatCurve=DataInterface() {propertyValues {Curve {keys [
          {arriveTangent=-1.25, interpMode=RCIM_Cubic, time=0, value=0}
          {arriveTangent=-13.5437, interpMode=RCIM_Cubic, leaveTangent=-13.5437, tangentMode=RCTM_User, time=0.0552, value=1.4953}
          {arriveTangent=-0.5791, interpMode=RCIM_Cubic, leaveTangent=-0.5791, time=0.1905, value=0.9925}
          {arriveTangent=-1.257, interpMode=RCIM_Cubic, leaveTangent=-1.257, time=0.9536, value=0.975}
          {interpMode=RCIM_Cubic, leaveTangent=-7.9488, time=0.98, value=0}
         ]}}}
      CurveIndex=$Emitter.NormalizedLoopAge
     )
    )) {isSetParameters=true}
   /SetVariables(Particles.SpritePos:NiagaraPosition=&RotateVectorByQuaternion(
     Vector To Rotate=&MakeVector(X=$Particles.dest_length, Y=0, Z=0)
     Quaternion=$Particles.CalculatedLineOrientation
    )) {isSetParameters=true}
   /SetVariables(Particles.Line_Length:NiagaraFloat=&VectorLength(Vector=Linked() {linkedVariable {name=Particles.SpritePos, type=NiagaraPosition}})) {isSetParameters=true}
   /SetVariables(Particles.MeshOrientation=$Particles.CalculatedLineOrientation) {isSetParameters=true}
   &Color(Color=&ColorFromCurve(
     ColorCurve=DataInterface() {propertyValues {
       AlphaCurve {keys [
         <at 0 0 RCIM_Cubic>
         {arriveTangent=1.0447, interpMode=RCIM_Cubic, leaveTangent=1.0447, time=0.0287, value=1.85}
         {arriveTangent=-1.9273, interpMode=RCIM_Cubic, leaveTangent=-1.9273, time=0.9688, value=1.0121}
         <at 0.9886 0 RCIM_Cubic>
        ]}
       BlueCurve {keys [<at 0.0549 1>]}
       GreenCurve {keys [<at 0.0549 0.5812>]}
       RedCurve {keys [<at 0.0549 0.3476>]}
      }}
     CurveIndex=$Emitter.NormalizedLoopAge
    )) {module=Color001}
   &SpriteRotationRate(Rotation Rate=50)
   &ScaleColor(
    Color Value To Scale=$Particles.Color
    Scale Alpha=&Multiply_Float(A=0.2, B=&Waveform("Waveform [1]"=::NewEnumerator12, "[1] Frequency"=12, "[1] Amplitude Min/Max"=<0.5 1.1>))
    ScaleA=true
    ScaleRGB=true
   )
  ]
  rendererClasses [&NiagaraMeshRendererProperties, &NiagaraSpriteRendererProperties]
  renderers [
   {rendererClass=&NiagaraMeshRendererProperties, rendererIndex=0}
   {rendererClass=&NiagaraSpriteRendererProperties, rendererIndex=1}
  ]
 }
 TorchGlow {
  emitterSpawnScript [/SetVariables() {isSetParameters=true}]
  emitterUpdateScript [
   &EmitterState()
   &SpawnBurst_Instantaneous(Spawn Count=1, Use Spawn Probability=false)
   /SetVariables(Emitter.emitter_reveal:NiagaraFloat=$System.ScanWidthTime) {isSetParameters=true}
  ]
  particleSpawnScript [
   &InitializeParticle(
    Color=&MakeCustomLinearColorFromBool(Bool=$User.bIsFrontEndPreview, True Color=<rgba 0.2629 0.6145 0.9896 0.5>, False Color=<rgba 0.2629 0.6145 0.9896 0.25>)
    Lifetime=$Emitter.CurrentLoopDuration
    Mesh Scale=<1.5 1.5 2.75>
    Mesh Scale Mode=::NewEnumerator4
    Sprite Size Mode=::NewEnumerator3
    Uniform Sprite Size=30
    UsePositionOffset=false
    Write Material Random=true
   )
   /SetVariables(Particles.MeshOrientation=&MakeQuatFromAxisAngle(Axis=<0 1 0>, AngleInDegrees=90)) {isSetParameters=true}
  ]
  particleUpdateScript [
   &ParticleState()
   &SolveForcesAndVelocity(Clamp Velocity=false, Limit Acceleration=false)
   /SetVariables(Particles.SoftGlowColor:LinearColor=&MultiplyLinearColor(Linear Color=$Particles.Color, Scale Factor=<1 1 1 0.025>)) {isSetParameters=true}
   &DynamicMaterialParameters(
    Index 0 Param 1=20
    Index 0 Param 2=0.1
    Index 1 Param 1=2
    Module.Param0WriteEnabled=true
    Module.Param0WriteEnabled1=true
    Module.Param1WriteEnabled=true
    Module.Param1WriteEnabled1=true
    Module.Param2WriteEnabled=true
    Module.Param2WriteEnabled1=true
    Module.Param3WriteEnabled=true
    Module.Param3WriteEnabled1=true
    Write Parameter Index 1=true
   )
   &Color(Color=&LerpLinearColors(StartColor=<rgba 1 0 0 10>, EndColor=<rgba 0 1 0 10>, LerpFactor=$System.ViewConeFalloffEval)) {enabled=false}
   &ScaleColor(
    Scale Alpha=&Multiply_Float(
     A=&Multiply_Float(
      A=&FloatFromCurve(
       FloatCurve=DataInterface() {propertyValues {Curve {keys [
           <at 0 0 RCIM_Cubic>
           {arriveTangent=2.3333, interpMode=RCIM_Cubic, leaveTangent=2.3333, time=0.1, value=0.975}
           {arriveTangent=0.9375, interpMode=RCIM_Cubic, leaveTangent=0.9375, time=0.3, value=0.7}
           {arriveTangent=0.0625, interpMode=RCIM_Cubic, leaveTangent=0.0625, time=0.5, value=1.35}
           {arriveTangent=-1.0625, interpMode=RCIM_Cubic, leaveTangent=-1.0625, time=0.7, value=0.725}
           {arriveTangent=-2.4167, interpMode=RCIM_Cubic, leaveTangent=-2.4167, time=0.9, value=0.925}
           <at 1 0 RCIM_Cubic>
          ]}}}
       CurveIndex=$Emitter.NormalizedLoopAge
      )
      B=&Waveform("Waveform [1]"=::NewEnumerator12, "[1] Frequency"=15, "[1] Amplitude Min/Max"=<0.5 1.1>)
     )
     B=&FN_Select_Float_from_Time_Of_Day(Morning=0.6, Day=0.6, Evening=1, Night=1)
    )
    ScaleA=true
    ScaleRGB=false
   )
   &ScaleMeshSize(Scale Factor=&MakeVector(X=$Emitter.emitter_reveal, Y=$Emitter.emitter_reveal, Z=1))
   &ScaleColor(
    Color Value To Scale=$Particles.Color
    Scale RGB=&VectorFromFloat(Value=&Lerp_Float(A=1, B=35, Alpha=$System.find_timeline))
    ScaleA=true
    ScaleRGB=true
   ) {module=ScaleColor001}
   &UpdateMeshOrientation(Rotation Rate=-0.125, Rotation Vector=<0 0 1>)
  ]
  rendererClasses [&NiagaraMeshRendererProperties]
  renderers [
   {rendererClass=&NiagaraMeshRendererProperties, rendererIndex=0}
   {rendererClass=&NiagaraMeshRendererProperties, rendererIndex=1}
  ]
 }
}
summary {
 systemName=NS_Emote_MedWayfare_Scanning
 emitters [
  {emitterName=TorchGlow, rendererClasses [&NiagaraMeshRendererProperties]}
  {emitterName=ScanningLines, rendererClasses [&NiagaraMeshRendererProperties, &NiagaraSpriteRendererProperties]}
  {bEnabled=false, emitterName=ScanResult, rendererClasses [&NiagaraSpriteRendererProperties]}
 ]
 userVariables [
  User.bIsFrontEnd {type=NiagaraBool}
  User.bIsFrontEndPreview {type=NiagaraBool}
  User.Particles_SpriteSize {type=Vector2f, description="", defaultValue {struct=Vector2f, value=<0 0>}}
 ]
}
