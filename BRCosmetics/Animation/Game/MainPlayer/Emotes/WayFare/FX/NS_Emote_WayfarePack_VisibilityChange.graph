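# Niagara one-shot flash for the Wayfare emote, named for the pack's visibility change.
# A brief glow sprite with an animated light and an expanding, distance-culled ring take their tint from User.PrimaryColor.
# Two crossed lens-flare streaks and a GPU burst of curl-noise embers use fixed blue colours instead.
# The DirectionalBurst and SimpleSpriteBurst emitters are disabled, as are all skeletal-mesh sampling modules.
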NiagaraSystem /BRCosmetics/Animation/Game/MainPlayer/Emotes/WayFare/FX/NS_Emote_WayfarePack_VisibilityChange
dependencies {
 usedDataInterfaces [&NiagaraDataInterfaceCurve, &NiagaraDataInterfaceColorCurve, &NiagaraDataInterfaceVector2DCurve]
 usedDynamicInputs []
 usedModules [&SystemState]
}
emitters {
 DirectionalBurst {
  bEnabled=false
  emitterUpdateScript [
   &EmitterState(Life Cycle Mode=::NewEnumerator1, Loop Behavior=::NewEnumerator1, Loop Duration=1, UseLoopDelay=false)
   &SpawnBurst_Instantaneous(Spawn Count=75, Use Spawn Probability=false, UseLoopCountLimit=false)
  ]
  particleSpawnScript [
   &InitializeParticle(
    Lifetime Max=1.75
    Lifetime Min=0.8
    Lifetime Mode=::NewEnumerator1
    Mass Max=2
    Mass Mode=::NewEnumerator3
    Sprite Size Max=<3 14>
    Sprite Size Min=<2 6>
    Sprite Size Mode=::NewEnumerator2
    UsePositionOffset=false
    Write Material Random=true
   )
   &AddVelocityInCone(
    Cone Angle=360
    Cone Axis=<0.9 0 0.4>
    Module.Use Velocity Falloff On Cone Axis=true
    Velocity Distribution Along Cone Axis=0.75
    Velocity Falloff Away From Cone Axis=0.333
    Velocity Strength=&RandomRangeFloat(Minimum=350, Maximum=750)
   )
  ]
  particleUpdateScript [
   &ParticleState()
   &GravityForce()
   &Drag(Drag=&RandomRangeFloat(Minimum=0.8, Maximum=1.2), Use Linear Drag=true, Use Rotational Drag=false)
   &ScaleColor(
    Scale Alpha=&FloatFromCurve(FloatCurve=DataInterface() {propertyValues {Curve {keys [<at -0.0046 1.0113 RCIM_Cubic>,<at 1.0011 0.002 RCIM_Cubic>]}}})
    ScaleA=true
    ScaleRGB=true
   )
   &SolveForcesAndVelocity(Clamp Velocity=false, Limit Acceleration=false)
   &ScaleSpriteSizeBySpeed(
    Max Scale Factor=<1.2 3.3>
    Min Scale Factor=<0.75 0.3>
    Sample Scale Factor By Curve=true
    Scale Factor Curve=DataInterface() {propertyValues {Curve {keys [
        {arriveTangent=2.865, interpMode=RCIM_Cubic, time=0.1247, value=-0.0006}
        {arriveTangent=4.1522, interpMode=RCIM_Cubic, leaveTangent=4.1522, tangentMode=RCTM_User, time=0.3948, value=0.4893}
        <at 0.9958 0.9974 RCIM_Cubic>
       ]}}}
    Velocity Threshold=1500
   )
  ]
  rendererClasses [&NiagaraSpriteRendererProperties]
  renderers [{rendererClass=&NiagaraSpriteRendererProperties}]
 }
 FN_AOE_Embers_Center {
  simTarget=GPUComputeSim
  emitterUpdateScript [
   &EmitterState()
   &SpawnRate(
    SpawnRate=&FloatFromCurve(FloatCurve=DataInterface() {propertyValues {Curve {keys [<at 0 100>,<at 0.15 100>,<at 0.2 0>]}}}, CurveIndex=$Emitter.NormalizedLoopAge)
    Use Spawn Probability=false
   ) {enabled=false}
   &SpawnBurst_Instantaneous(Spawn Count=&RandomRangeInt(Minimum=10, Maximum=30), Use Spawn Probability=false)
  ]
  particleSpawnScript [
   /Niagara/Modules/Spawn/Initialization/InitializeParticle(
    Color=<rgba 0.0439 0.6212 1 1>
    Lifetime=&UniformRangedFloat(Minimum=0.7, Maximum=1)
    Module.Write Color=true
    Module.Write Lifetime=true
    Module.Write Mass=false
    Module.Write Material Random=true
    Module.Write Position=false
    Module.Write Scale=false
    Module.Write SpriteRotation=true
    Module.Write SpriteSize=true
    Position=&SimulationPosition()
    Sprite Rotation=&UniformRangedFloat(Minimum=-180, Maximum=180)
    Sprite Size=&Vector2DFromFloat(Value=&UniformRangedFloat(Minimum=1, Maximum=1))
   )
   &SkeletalMeshLocation(Bone Sampling Mode=::NewEnumerator4, MaskSpawnGroup=false, Skeletal Mesh=$User.SkeletalMesh) {enabled=false, module=SkeletalMeshLocation001}
   &AddVelocity(Velocity=&MakeVector(X=0, Y=0, Z=&UniformRangedFloat(Minimum=25, Maximum=75))) {enabled=false}
   /Niagara/Modules/Spawn/Location/SphereLocation(MaskSpawnGroup=false, Module.Override Randomness=false, Sphere Radius=10, Surface Only=false)
   &AddVelocityFromPoint(
    Module.Use Falloff Distance=false
    Velocity Strength=&Multiply_Float(A=&RandomRangeFloat(Minimum=25, Maximum=100), B=2)
   )
   &ApplyInitialForces()
  ]
  particleUpdateScript [
   &ParticleState()
   /Niagara/Modules/Update/Forces/CurlNoiseForce(
    Noise Frequency=0.58
    Noise Strength=&FloatFromCurve(
     FloatCurve=DataInterface() {propertyValues {Curve {keys [<at 0 0 RCIM_Cubic>,<at 0.3 0.875 RCIM_Cubic>,<at 1 0 RCIM_Cubic>]}}}
     Scale Curve=500
    )
    Pan Noise=true
    Pan Noise Field=<0 0 1>
   )
   &AccelerationForce(Acceleration=&RandomRangeVector(Minimum=<0 0 -25>, Maximum=<1 1 25>))
   &Drag(Drag=8, Use Linear Drag=true, Use Rotational Drag=false)
   &ScaleColor(
    Scale Alpha=&FloatFromCurve(FloatCurve=DataInterface() {propertyValues {Curve {keys [
         {arriveTangent=11.9433, interpMode=RCIM_Cubic, leaveTangent=11.9433, tangentMode=RCTM_User, time=0, value=0}
         <at 0.2 1 RCIM_Cubic>
         <at 1 0 RCIM_Cubic>
        ]}}})
    Scale RGB=&MakeVectorFromLinearColorRGB(Color=&ColorFromCurve(ColorCurve=DataInterface() {propertyValues {
        AlphaCurve {keys [<at 0 1 RCIM_Cubic>]}
        BlueCurve {keys [<at 0 4 RCIM_Cubic>,<at 1 1 RCIM_Cubic>]}
        GreenCurve {keys [<at 0 3.9667 RCIM_Cubic>,<at 1 1 RCIM_Cubic>]}
        RedCurve {keys [<at 0 3.6 RCIM_Cubic>,<at 1 1 RCIM_Cubic>]}
       }}))
    ScaleA=true
    ScaleRGB=true
   )
   &DynamicMaterialParameters(
    Index 0 Param 1=&FloatFromCurve(FloatCurve=DataInterface() {propertyValues {Curve {keys [<at 0 0>,<at 1 1>]}}})
    Index 0 Param 2=1
    Index 0 Param 3=1
    Index 0 Param 4=1
    Module.Param0WriteEnabled=true
    Module.Param1WriteEnabled=true
    Module.Param2WriteEnabled=true
    Module.Param3WriteEnabled=true
   )
   &ScaleSpriteSize(Scale Factor=&Vector2DFromFloat(Value=&FloatFromCurve(FloatCurve=DataInterface() {propertyValues {Curve {keys [
          {arriveTangent=9.1531, interpMode=RCIM_Cubic, leaveTangent=9.1531, tangentMode=RCTM_User, time=0, value=0.02}
          {arriveTangent=0.48, interpMode=RCIM_Cubic, leaveTangent=0.48, time=0.18, value=0.99}
          <at 1 0.5 RCIM_Cubic>
         ]}}})))
   &SolveForcesAndVelocity(Clamp Velocity=false, Limit Acceleration=false)
   &ScaleSpriteSizeBySpeed(Max Scale Factor=<1 2>)
  ]
  rendererClasses [&NiagaraSpriteRendererProperties]
  renderers [{rendererClass=&NiagaraSpriteRendererProperties}]
 }
 Glow {
  emitterSpawnScript [/SetVariables(Emitter.GlowDepthFade=50) {isSetParameters=true}]
  emitterUpdateScript [
   &EmitterState()
   &SpawnBurst_Instantaneous(Spawn Count=1, Use Spawn Probability=false)
  ]
  particleSpawnScript [
   &FN_SetSpawnGroup(Number of Spawn Groups=1)
   /Niagara/Modules/Spawn/Initialization/V2/InitializeParticle(
    Color=$User.PrimaryColor
    Lifetime=0.2
    Sprite Rotation Mode=::NewEnumerator0
    Sprite Size Mode=::NewEnumerator3
    Uniform Sprite Size=80
    UsePositionOffset=true
    Write Material Random=true
   )
   /SetVariables(Particles.FlareColor=<rgba 1 1 1 25>) {enabled=false, isSetParameters=true}
   /SetVariables(Particles.FlareSize:Vector2f=<40 10>) {enabled=false, isSetParameters=true}
   /SetVariables(Particles.FlareAlignement:Vector3f=$Engine.Owner.SystemZAxis) {enabled=false, isSetParameters=true}
   /SetVariables(Particles.FlareFacing:Vector3f=$Engine.Owner.SystemYAxis) {enabled=false, isSetParameters=true}
   &CameraOffset(Camera Offset Amount=10)
  ]
  particleUpdateScript [
   &ParticleState()
   &ScaleColor(
    Linear Color Curve=DataInterface() {propertyValues {
      AlphaCurve {keys [<at 0 0>,<at 0.2569 1>,<at 1 0>]}
      BlueCurve {keys [<at 0.2606 10>,<at 1 0>]}
      GreenCurve {keys [<at 0.2606 10>,<at 1 0>]}
      RedCurve {keys [<at 0.2606 10>,<at 1 0>]}
     }}
    Scale Mode=::NewEnumerator2
   )
   &DynamicMaterialParameters(
    Index 0 Param 1=1
    Module.Param0WriteEnabled=true
    Module.Param0WriteEnabled1=true
    Module.Param1WriteEnabled=true
    Module.Param1WriteEnabled1=true
    Module.Param2WriteEnabled=true
    Module.Param2WriteEnabled1=true
    Module.Param3WriteEnabled=true
    Module.Param3WriteEnabled1=true
    Write Parameter Index 1=true
   )
   &ScaleSpriteSize(
    Uniform Curve Scale=0.7
    Uniform Curve Sprite Scale=DataInterface() {propertyValues {Curve {keys [{tangentMode=RCTM_User, time=0, value=0}, <at 0.1 1>, {tangentMode=RCTM_User, time=1, value=1}]}}}
   )
   &FN_Gen5_Light(
    Attenutation:NiagaraFloat=&FloatFromCurve(
     FloatCurve=DataInterface() {propertyValues {Curve {keys [
         {arriveTangent=9.0725, interpMode=RCIM_Cubic, leaveTangent=9.0725, tangentMode=RCTM_User, time=0, value=0}
         {arriveTangent=0.173, interpMode=RCIM_Cubic, leaveTangent=0.173, tangentMode=RCTM_User, time=0.1408, value=0.9847}
         {arriveTangent=-2.2361, interpMode=RCIM_Cubic, leaveTangent=-2.2361, tangentMode=RCTM_User, time=1, value=0.25}
        ]}}}
     Scale Curve=100
    )
    Exponent=2.5
    Intensity=2000
   )
   &FN_BoneFlow(Height:NiagaraFloat=$Particles.NormalizedAge, SkeletalMesh:NiagaraDataInterfaceSkeletalMesh=$User.SkeletalMesh) {enabled=false}
   &CameraOffset(Camera Offset Amount=20) {module=CameraOffset001}
  ]
  rendererClasses [&NiagaraSpriteRendererProperties, &NiagaraLightRendererProperties]
  renderers [
   {rendererClass=&NiagaraSpriteRendererProperties}
   {rendererClass=&NiagaraLightRendererProperties, rendererIndex=1}
  ]
 }
 Lens_Flares {
  emitterUpdateScript [
   &EmitterState()
   &SpawnBurst_Instantaneous(Spawn Count=2, Use Spawn Probability=false, UseLoopCountLimit=false)
  ]
  particleSpawnScript [
   &FN_SetSpawnGroup(Number of Spawn Groups=2)
   &InitializeParticle(
    Color=&MultiplyLinearColor(
     Linear Color=<rgba 0.1146 0.4559 1 1>
     Scale Factor=&MakeVector4(
      X=1
      Y=1
      Z=1
      W=&MakeCustomFloatFromBool(
       Bool=&SetBoolByIntComparison(A=&SpawnGroup(), B=0, Comparison Type=::NewEnumerator4)
       True Float=0.8
       False Float=0.2
      )
     )
    )
    Lifetime=0.4
    Position Offset Coordinate Space=::Local
    Sprite Rotation Angle=&MaskFloatBySpawnGroup(Masked Float=90, Passthrough Float=0, Spawn Group Mask=1)
    Sprite Rotation Mode=::NewEnumerator0
    Sprite Size=&Multiply_Vector2D_ByFloat(
     Float=0.6
     Vector2D=&MakeCustomVector2DFromBool(
      Bool=&SetBoolByIntComparison(A=&SpawnGroup(), B=0, Comparison Type=::NewEnumerator4)
      True Vector2D=<60 500>
      False Vector2D=<60 492.4811>
     )
    )
    Sprite Size Mode=::NewEnumerator4
    UsePositionOffset=true
    Write Material Random=true
   )
  ]
  particleUpdateScript [
   &ParticleState()
   &ScaleColor(
    Scale Alpha=&FloatFromCurve(FloatCurve=DataInterface() {propertyValues {Curve {keys [
         {arriveTangent=-3.0287, interpMode=RCIM_Cubic, time=0, value=1}
         {arriveTangent=-1, interpMode=RCIM_Cubic, leaveTangent=-1, time=0.5, value=0.05}
         <at 1 0 RCIM_Cubic>
        ]}}})
    ScaleA=true
    ScaleRGB=false
   )
   &CameraOffset(Camera Offset Amount=-22)
   &ScaleSpriteSize(
    Non-Uniform Curve Sprite Scale=DataInterface() {propertyValues {
      XCurve {keys [<at 1 1 RCIM_Cubic>]}
      YCurve {keys [
        <at 0 0.5 RCIM_Cubic>
        {arriveTangent=0.75, interpMode=RCIM_Cubic, leaveTangent=0.75, time=0.1, value=2.625}
        {arriveTangent=-1.8056, interpMode=RCIM_Cubic, leaveTangent=-1.8056, time=0.5, value=0.875}
        <at 1 1 RCIM_Cubic>
       ]}
     }}
    Scale Sprite Size Mode=::NewEnumerator3
   )
  ]
  rendererClasses [&NiagaraSpriteRendererProperties]
  renderers [{rendererClass=&NiagaraSpriteRendererProperties}]
 }
 Ring {
  emitterSpawnScript [/SetVariables(Emitter.GlowDepthFade=25) {isSetParameters=true}]
  emitterUpdateScript [
   &EmitterState(
    Enable Distance Culling=true
    Life Cycle Mode=::NewEnumerator1
    Loop Behavior=::NewEnumerator1
    Loop Delay=0.0333
    Loop Duration=1
    Max Distance Response=::KillImmediately
    MaxDistance=3500
    Module.bUseMaxDistance=true
    Module.bUseMinDistance=false
    Scalability Mode=::NewEnumerator1
    UseLoopDelay=true
   )
   &SpawnBurst_Instantaneous(Spawn Count=1, Use Spawn Probability=false)
   &SpawnBurst_Instantaneous(Spawn Time=0.1, Use Spawn Probability=false, UseLoopCountLimit=false) {enabled=false, module=SpawnBurst_Instantaneous001}
  ]
  particleSpawnScript [&InitializeParticle(
    Color=$User.PrimaryColor
    Lifetime=0.2
    Sprite Rotation Mode=::NewEnumerator2
    Sprite Size Mode=::NewEnumerator3
    Uniform Sprite Size=75
    UsePositionOffset=false
    Write Material Random=true
   )]
  particleUpdateScript [
   &ParticleState()
   &ScaleColor(
    Scale Alpha=&FloatFromCurve(FloatCurve=DataInterface() {propertyValues {Curve {keys [<at 0 0.5>,<at 1 0>]}}})
    Scale RGB=&VectorFromFloat(Value=&FloatFromCurve(FloatCurve=DataInterface() {propertyValues {Curve {keys [<at 0 1 RCIM_Cubic>, {arriveTangent=-1, interpMode=RCIM_Cubic, leaveTangent=-1, time=0.1, value=8}, <at 1 0 RCIM_Cubic>]}}}))
    ScaleA=true
    ScaleRGB=true
   )
   &ScaleSpriteSize(Scale Factor=&Vector2DFromFloat(Value=&FloatFromCurve(FloatCurve=DataInterface() {propertyValues {Curve {keys [{arriveTangent=2.7513, interpMode=RCIM_Cubic, leaveTangent=2.7513, tangentMode=RCTM_User, time=0, value=0}, <at 1 1>]}}})))
   /SetVariables(
    Particles.FlareColor=&MultiplyLinearColor(Linear Color=$User.PrimaryColor, Scale Factor=&MakeVector4_FromFloat(Float=1))
    Particles.FlareSize:Vector2f=&Multiply_Vector2D_ByFloat(
     Float=&FloatFromCurve(
      FloatCurve=DataInterface() {propertyValues {Curve {keys [
          {interpMode=RCIM_Cubic, tangentMode=RCTM_User, time=0, value=0}
          {interpMode=RCIM_Cubic, tangentMode=RCTM_User, time=0.3, value=0.95}
          {tangentMode=RCTM_User, time=1, value=0}
         ]}}}
      Scale Curve=1.25
     )
     Vector2D=$Particles.Initial.SpriteSize
    )
    Particles.FlashCameraOffset=40
   ) {isSetParameters=true}
   &SkeletalMeshLocation(
    Bone Sampling Mode=::NewEnumerator4
    MaskSpawnGroup=false
    Sampled Position Offset Coordinate Space:ENiagaraSkelMeshTransforms=::NewEnumerator2
    Skeletal Mesh=$User.SkeletalMesh
   ) {enabled=false}
   &CameraOffset(Camera Offset Amount=40)
   &DynamicMaterialParameters(
    Index 0 Param 1=1
    Index 0 Param 2=1
    Index 0 Param 3=1
    Module.Param0WriteEnabled=true
    Module.Param1WriteEnabled=true
    Module.Param2WriteEnabled=true
    Module.Param3WriteEnabled=true
   )
  ]
  rendererClasses [&NiagaraSpriteRendererProperties]
  renderers [
   {rendererClass=&NiagaraSpriteRendererProperties}
   {rendererClass=&NiagaraSpriteRendererProperties, rendererIndex=1}
  ]
 }
 SimpleSpriteBurst {
  bEnabled=false
  emitterUpdateScript [
   &EmitterState(Life Cycle Mode=::NewEnumerator1, Loop Behavior=::NewEnumerator1, Loop Duration=2, UseLoopDelay=false)
   &SpawnBurst_Instantaneous(Use Spawn Probability=false, UseLoopCountLimit=false)
  ]
  particleSpawnScript [&InitializeParticle(
    Color=<rgba 0.0749 1.5642 4 4>
    Lifetime=0.5
    Sprite Size Mode=::NewEnumerator3
    Uniform Sprite Size=50
    UsePositionOffset=false
    Write Material Random=true
   )]
  particleUpdateScript [
   &ParticleState()
   &ScaleColor(
    Scale Alpha=&FloatFromCurve(FloatCurve=DataInterface() {propertyValues {Curve {keys [<at -0.0046 1.0113 RCIM_Cubic>,<at 1.0011 0.002 RCIM_Cubic>]}}})
    ScaleA=true
    ScaleRGB=false
   )
   &SolveForcesAndVelocity(Clamp Velocity=false, Limit Acceleration=false)
   &ScaleSpriteSize(Uniform Curve Sprite Scale=DataInterface() {propertyValues {Curve {keys [
        <at 0 0 RCIM_Cubic>
        {arriveTangent=1.4583, interpMode=RCIM_Cubic, leaveTangent=1.4583, time=0.2, value=1.875}
        {arriveTangent=-1.3889, interpMode=RCIM_Cubic, leaveTangent=-1.3889, time=0.6, value=0.875}
        <at 1.1 0.625 RCIM_Cubic>
       ]}}})
   &SpriteRotationRate(Rotation Rate=25)
   &CameraOffset(Camera Offset Amount=&FloatFromCurve(FloatCurve=DataInterface() {propertyValues {Curve {keys [<at 0 1>,<at 1 0>]}}}, Scale Curve=50))
  ]
  rendererClasses [&NiagaraSpriteRendererProperties]
  renderers [{rendererClass=&NiagaraSpriteRendererProperties}]
 }
}
summary {
 systemName=NS_Emote_WayfarePack_VisibilityChange
 emitters [
  {bEnabled=false, emitterName=DirectionalBurst, rendererClasses [&NiagaraSpriteRendererProperties]}
  {bEnabled=false, emitterName=SimpleSpriteBurst, rendererClasses [&NiagaraSpriteRendererProperties]}
  {emitterName=FN_AOE_Embers_Center, simTarget=GPUComputeSim, rendererClasses [&NiagaraSpriteRendererProperties]}
  {emitterName=Glow, rendererClasses [&NiagaraSpriteRendererProperties, &NiagaraLightRendererProperties]}
  {emitterName=Ring, rendererClasses [&NiagaraSpriteRendererProperties]}
  {emitterName=Lens_Flares, rendererClasses [&NiagaraSpriteRendererProperties]}
 ]
 userVariables [User.PrimaryColor {type=LinearColor, description="", defaultValue {struct=LinearColor, value=<rgba 0.1197 0.4173 1 1>}}]
}
